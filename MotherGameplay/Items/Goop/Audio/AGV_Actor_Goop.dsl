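; Audio volume actor for Goop that muffles its edge sound as the box's inside state changes.
; On each AudioShapeBox inside-state change, it enables a low-pass filter on the shape's edge audio component.
; Entering plays a timeline from the start, lowering the filter cutoff and feeding an "Inside" float parameter to the sound.
; Exiting reverses that timeline from the end to fade the effect back out.
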(event OnInsideStateChanged(AudioShapeBox)
  (params (bIsInside :type Boolean))
  (bind _audioshapebox (Variables|Default|GetAudioShapeBox))
  (bind _returnvalue (Control|GetEdgeAudioComponent _audioshapebox))
  ;[1 1 1] Apply low pass filter to edge audio when entering volume
  (Audio|Components|Audio|SetLowPassFilterEnabled _returnvalue true :!class /Script/Engine.AudioComponent)
  (if bIsInside
    (bind (_direction _inside) (Timeline|Timeline_0 PlayFromStart)
      (:Update
        (Audio|Components|Audio|SetLowPassFilterFrequency _returnvalue (Math|Float|MapRangeClamped _inside 0.0 1.0 2e+04 5e+02) :!class /Script/Engine.AudioComponent)
        (Audio|Parameter|SetFloatParameter _returnvalue "Inside" _inside :!class /Script/AudioExtensions.AudioParameterControllerInterface)))
    (else
      (Timeline|Timeline_0 ReverseFromEnd))))

(fn ConstructionScript)
