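# Sound cue for the Goop item's looping body effect, spatialized with a Goop traversal attenuation asset.
# A looping wave player feeds two chained continuous modulators: a Speed parameter scales pitch and volume, then an IsFalling parameter can boost volume.
# Virtualization is disabled, and a concurrency override caps simultaneous instances, stopping the quietest.
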SoundCue /MotherGameplay/Items/Goop/Audio/Goop_Body_FX_Loop_Cue
root=ModulatorContinuous_1
nodes [
 {
  id=ModulatorContinuous_0
  children [WavePlayer_0]
  properties {
   pitchModulationParams {default=0, minOutput=0.75, parameterName=Speed}
   volumeModulationParams {default=0, minOutput=0.1, parameterName=Speed}
  }
 }
 {
  id=ModulatorContinuous_1
  children [ModulatorContinuous_0]
  properties {
   pitchModulationParams {default=0, minOutput=1, parameterName=IsFalling}
   volumeModulationParams {default=0, maxOutput=1.5, minOutput=1, parameterName=IsFalling}
  }
 }
 {id=WavePlayer_0, properties {bLooping=true, soundWaveAssetPtr=&/Goop_Body_FX_Loop}}
]
properties {
 attenuationSettings=&/Fort_Goop_Traversal_Att
 bOverrideConcurrency=true
 maxDistance=3600
 soundClassObject=/Game/Sounds/FortSoundClassesAndModes/Fort_AGV
 virtualizationMode=Disabled
 volumeMultiplier=2.25
 concurrencyOverrides {maxCount=3, resolutionRule=StopQuietest}
}
