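# Sound cue that plays when aiming starts with the equipped Goop item, going by its name and wave names.
# It picks one of two aim-start waves with equal weight, then a modulator randomizes volume and slightly raises pitch.
# It uses weapon-reload attenuation and the AGV sound class, with a boosted volume multiplier.
# An owner-limited concurrency override allows one instance at a time and prevents new plays from interrupting it.
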SoundCue /MotherGameplay/Items/Goop/Audio/Goop_Equip_Aim_Cue
root=Modulator_1
nodes [
 {id=Modulator_1, children [Random_0], properties {pitchMax=1.2, pitchMin=1, volumeMax=1, volumeMin=0.3}}
 {id=Random_0, children [WavePlayer_2,WavePlayer_3], properties {bShouldExcludeFromBranchCulling=true, weights [1,1]}}
 {id=WavePlayer_2, properties {soundWaveAssetPtr=&/Goop_Equip_Aim_Start_01}}
 {id=WavePlayer_3, properties {soundWaveAssetPtr=&/Goop_Equip_Aim_Start_02}}
]
properties {
 attenuationSettings=&Fort_Weapon_Reload_Att
 bOverrideConcurrency=true
 maxDistance=2250
 soundClassObject=/Game/Sounds/FortSoundClassesAndModes/Fort_AGV
 volumeMultiplier=6.72
 concurrencyOverrides {bLimitToOwner=true, maxCount=1, resolutionRule=PreventNew, retriggerTime=1}
}
