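# Sound cue named for equipping the Goop item, currently empty with no nodes and no root, so it references no audio.
# It still sets the shared Fort weapon reload attenuation, the Fort_AGV sound class, a slight volume boost and a maximum audible distance.
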SoundCue /MotherGameplay/Items/Goop/Audio/Goop_Equip_Cue
root=null
nodes []
properties {
 attenuationSettings=&Fort_Weapon_Reload_Att
 maxDistance=2100
 soundClassObject=/Game/Sounds/FortSoundClassesAndModes/Fort_AGV
 volumeMultiplier=1.12
}
