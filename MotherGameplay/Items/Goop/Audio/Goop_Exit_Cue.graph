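# Sound cue for the whoosh heard on exiting Goop, apparently part of Goop traversal given its enter/exit attenuation asset.
# A Random node picks one of six equally weighted whoosh waves and feeds a Modulator at default settings for per-play pitch and volume variation.
# The cue boosts volume, uses the Fort_AGV sound class, and overrides concurrency to cap simultaneous instances, stopping the quietest.
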SoundCue /MotherGameplay/Items/Goop/Audio/Goop_Exit_Cue
root=Modulator_0
nodes [
 {id=Modulator_0, children [Random_0]}
 {id=Random_0, children [WavePlayer_0,WavePlayer_1,WavePlayer_2,WavePlayer_3,WavePlayer_4,WavePlayer_5], properties {weights [1,1,1,1,1,1]}}
 {id=WavePlayer_0, properties {soundWaveAssetPtr=&/Goop_Exit_Whoosh_01}}
 {id=WavePlayer_1, properties {soundWaveAssetPtr=&/Goop_Exit_Whoosh_02}}
 {id=WavePlayer_2, properties {soundWaveAssetPtr=&/Goop_Exit_Whoosh_03}}
 {id=WavePlayer_3, properties {soundWaveAssetPtr=&/Goop_Exit_Whoosh_04}}
 {id=WavePlayer_4, properties {soundWaveAssetPtr=&/Goop_Exit_Whoosh_05}}
 {id=WavePlayer_5, properties {soundWaveAssetPtr=&/Goop_Exit_Whoosh_06}}
]
properties {
 attenuationSettings=&/Fort_Goop_Traversal_Enter_Exit_Att
 bOverrideConcurrency=true
 maxDistance=6300
 soundClassObject=/Game/Sounds/FortSoundClassesAndModes/Fort_AGV
 volumeMultiplier=1.62
 concurrencyOverrides {maxCount=3, resolutionRule=StopQuietest}
}
