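# Looping exterior ambient sound cue for the Goop item, heard when the Inside parameter is low.
# Close and distant loops each pass through a Modulator and their own attenuation asset, then a Mixer blends them.
# A continuous modulator at the root reads Inside, fading volume out and lowering pitch as the value rises, so the exterior layer recedes inside.
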SoundCue /MotherGameplay/Items/Goop/Audio/Goop_Exterior_Loop_Cue
root=ModulatorContinuous_0
nodes [
 {id=Attenuation_0, children [Modulator_0], properties {attenuationSettings=&/Fort_Goop_Energy_Loop_Close_Att}}
 {
  id=Attenuation_1
  children [Modulator_1]
  properties {attenuationSettings=&/Fort_Goop_Energy_Loop_Distant_Att}
 }
 {id=Mixer_0, children [Attenuation_0, Attenuation_1], properties {inputVolume [2,2]}}
 {
  id=ModulatorContinuous_0
  children [Mixer_0]
  properties {
   pitchModulationParams {default=0, maxOutput=0.5, minOutput=1, parameterName=Inside}
   volumeModulationParams {default=0, maxOutput=0, minOutput=1, parameterName=Inside}
  }
 }
 {id=Modulator_0, children [WavePlayer_0]}
 {id=Modulator_1, children [WavePlayer_1]}
 {id=WavePlayer_0, properties {bLooping=true, soundWaveAssetPtr=&/Goop_Exterior_Close_Loop}}
 {id=WavePlayer_1, properties {bLooping=true, soundWaveAssetPtr=&/Goop_Exterior_Distant_Loop}}
]
properties {maxDistance=4000, soundClassObject=/Game/Sounds/FortSoundClassesAndModes/Fort_AGV, volumeMultiplier=1.08}
