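# Sound cue that plays the looping interior ambience for the Goop item.
# A single looping wave player feeds Goop_Interior_Ambient_Loop straight to the output at boosted volume, under the indoor ambient sound class.
# A concurrency override allows one instance at a time, stopping the oldest when another starts.
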SoundCue /MotherGameplay/Items/Goop/Audio/Goop_Interior_Loop_Cue
root=WavePlayer_2
nodes [{id=WavePlayer_2, properties {bLooping=true, soundWaveAssetPtr=&/Goop_Interior_Ambient_Loop}}]
properties {
 bOverrideConcurrency=true
 maxDistance=2097152
 soundClassObject=&Fort_Ambient_Indoors_Athena
 volumeMultiplier=2.506
 concurrencyOverrides {maxCount=1, resolutionRule=StopOldest}
}
