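# Sound cue for the whoosh played when the Goop item performs a navigation jump.
# A random node picks one of six equally weighted whoosh variations, and a modulator varies its pitch before it feeds a mixer whose second input is empty.
# Playback uses Goop traversal attenuation and the AGV sound class, with virtualization disabled and a per-owner concurrency limit that stops the quietest voice.
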SoundCue /MotherGameplay/Items/Goop/Audio/Goop_Nav_Jump_Cue
root=Mixer_0
nodes [
 {id=Mixer_0, children [Modulator_1,null], properties {inputVolume [1.65,0.65]}}
 {id=Modulator_1, children [Random_1], properties {pitchMax=1.5}}
 {
  id=Random_1
  children [WavePlayer_6, WavePlayer_7, WavePlayer_8, WavePlayer_9, WavePlayer_10, WavePlayer_11]
  properties {weights [1,1,1,1,1,1]}
 }
 {id=WavePlayer_10, properties {soundWaveAssetPtr=&/Goop_Jump_Whoosh_05}}
 {id=WavePlayer_11, properties {soundWaveAssetPtr=&/Goop_Jump_Whoosh_06}}
 {id=WavePlayer_6, properties {soundWaveAssetPtr=&/Goop_Jump_Whoosh_01}}
 {id=WavePlayer_7, properties {soundWaveAssetPtr=&/Goop_Jump_Whoosh_02}}
 {id=WavePlayer_8, properties {soundWaveAssetPtr=&/Goop_Jump_Whoosh_03}}
 {id=WavePlayer_9, properties {soundWaveAssetPtr=&/Goop_Jump_Whoosh_04}}
]
properties {
 attenuationSettings=&/Fort_Goop_Traversal_Att
 bOverrideConcurrency=true
 maxDistance=3600
 soundClassObject=/Game/Sounds/FortSoundClassesAndModes/Fort_AGV
 virtualizationMode=Disabled
 volumeMultiplier=2.25
 concurrencyOverrides {bLimitToOwner=true, maxCount=4, resolutionRule=StopQuietest, voiceStealReleaseTime=0.1}
}
