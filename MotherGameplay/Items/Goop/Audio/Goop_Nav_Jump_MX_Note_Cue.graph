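# Sound cue for a musical whoosh note, apparently played on Goop navigation jumps, routed to the Fort_AGV sound class with Goop traversal attenuation.
# A random node picks one of four equally weighted whoosh music waves, then a modulator varies volume while pitch stays fixed.
# The mixer's first input is empty, so only this branch sounds.
# A concurrency override allows one instance per owner, preventing new ones while it plays, and virtualization is disabled.
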SoundCue /MotherGameplay/Items/Goop/Audio/Goop_Nav_Jump_MX_Note_Cue
root=Mixer_0
nodes [
 {id=Mixer_0, children [null,Modulator_2], properties {inputVolume [1.65,0.65]}}
 {id=Modulator_2, children [Random_2], properties {pitchMax=1, pitchMin=1, volumeMax=1, volumeMin=0.6}}
 {
  id=Random_2
  children [WavePlayer_12, WavePlayer_13, WavePlayer_14, WavePlayer_15]
  properties {bShouldExcludeFromBranchCulling=true, weights [1,1,1,1]}
 }
 {id=WavePlayer_12, properties {soundWaveAssetPtr=&/Goop_Jump_Whoosh_Music_01}}
 {id=WavePlayer_13, properties {soundWaveAssetPtr=&/Goop_Jump_Whoosh_Music_02}}
 {id=WavePlayer_14, properties {soundWaveAssetPtr=&/Goop_Jump_Whoosh_Music_03}}
 {id=WavePlayer_15, properties {soundWaveAssetPtr=&/Goop_Jump_Whoosh_Music_04}}
]
properties {
 attenuationSettings=&/Fort_Goop_Traversal_Att
 bOverrideConcurrency=true
 maxDistance=3600
 soundClassObject=/Game/Sounds/FortSoundClassesAndModes/Fort_AGV
 virtualizationMode=Disabled
 volumeMultiplier=10.08
 concurrencyOverrides {bLimitToOwner=true, maxCount=1, resolutionRule=PreventNew}
}
