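# Sound cue for landing from Goop navigation, judging by its name and traversal attenuation, layering a musical whoosh with a plain whoosh.
# A Mixer blends a random pick of four musical variants, excluded from branch culling, with a quieter random pick of four plain whooshes that gets pitch variation from a Modulator.
# It uses its own concurrency limit, stopping the quietest instance, with virtualization disabled.
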SoundCue /MotherGameplay/Items/Goop/Audio/Goop_Nav_Land_Cue
root=Mixer_0
nodes [
 {id=Mixer_0, children [Random_1,Modulator_1], properties {inputVolume [1.1,0.45]}}
 {id=Modulator_1, children [Random_2], properties {pitchMax=1.3, pitchMin=0.9}}
 {
  id=Random_1
  children [WavePlayer_6,WavePlayer_7,WavePlayer_8,WavePlayer_9]
  properties {bShouldExcludeFromBranchCulling=true, weights [1,1,1,1]}
 }
 {id=Random_2, children [WavePlayer_10, WavePlayer_11, WavePlayer_12, WavePlayer_13], properties {weights [1,1,1,1]}}
 {id=WavePlayer_10, properties {soundWaveAssetPtr=&/Goop_Land_Whoosh_01}}
 {id=WavePlayer_11, properties {soundWaveAssetPtr=&/Goop_Land_Whoosh_02}}
 {id=WavePlayer_12, properties {soundWaveAssetPtr=&/Goop_Land_Whoosh_03}}
 {id=WavePlayer_13, properties {soundWaveAssetPtr=&/Goop_Land_Whoosh_04}}
 {id=WavePlayer_6, properties {soundWaveAssetPtr=&/Goop_Land_Whoosh_Music_04}}
 {id=WavePlayer_7, properties {soundWaveAssetPtr=&/Goop_Land_Whoosh_Music_01}}
 {id=WavePlayer_8, properties {soundWaveAssetPtr=&/Goop_Land_Whoosh_Music_02}}
 {id=WavePlayer_9, properties {soundWaveAssetPtr=&/Goop_Land_Whoosh_Music_03}}
]
properties {
 attenuationSettings=&/Fort_Goop_Traversal_Att
 bOverrideConcurrency=true
 maxDistance=3600
 soundClassObject=/Game/Sounds/FortSoundClassesAndModes/Fort_AGV
 virtualizationMode=Disabled
 volumeMultiplier=4.144
 concurrencyOverrides {maxCount=3, resolutionRule=StopQuietest}
}
