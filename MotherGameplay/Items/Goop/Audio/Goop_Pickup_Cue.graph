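# Goop_Pickup_Cue is the sound cue for picking up the Goop item.
# A Modulator node at the root feeds from a single wave player of the Goop_Pickup_01 sound wave.
# It uses the shared weapon pickup attenuation and loot collect sound class.
# A concurrency override allows one instance at a time and blocks new plays while one is active.
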SoundCue /MotherGameplay/Items/Goop/Audio/Goop_Pickup_Cue
root=Modulator_0
nodes [{id=Modulator_0, children [WavePlayer_4]}, {id=WavePlayer_4, properties {soundWaveAssetPtr=&/Goop_Pickup_01}}]
properties {
 attenuationSettings=&Fort_Weapon_Pickup_Att
 bOverrideConcurrency=true
 maxDistance=3500
 soundClassObject=&Fort_Loot_Collect
 volumeMultiplier=0.498
 concurrencyOverrides {maxCount=1, resolutionRule=PreventNew}
}
