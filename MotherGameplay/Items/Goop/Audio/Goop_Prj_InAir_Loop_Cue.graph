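# Sound cue for the Goop projectile's in-flight audio.
# A Mixer layers a looping in-air sound, faded in by an Enveloper, with a throw whoosh and a modulated random pick from six distant whip variants.
# It uses shared launcher-projectile attenuation and overrides concurrency to cap simultaneous instances, favoring nearer ones and scaling volume with attack and release times.
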SoundCue /MotherGameplay/Items/Goop/Audio/Goop_Prj_InAir_Loop_Cue
root=Mixer_0
nodes [
 {
  id=Enveloper_0
  children [WavePlayer_0]
  properties {
   pitchCurve {editorCurveData {keys [{time=0}]}}
   volumeCurve {editorCurveData {keys [{arriveTangent=0.444, interpMode=RCIM_Cubic, leaveTangent=0.444, tangentMode=RCTM_User, time=0, value=0}, <at 1 1>]}}
  }
 }
 {id=Mixer_0, children [Enveloper_0,WavePlayer_1,Modulator_0], properties {inputVolume [1,0.8,0.8]}}
 {id=Modulator_0, children [Random_0]}
 {id=Random_0, children [WavePlayer_2,WavePlayer_3,WavePlayer_4,WavePlayer_5,WavePlayer_6,WavePlayer_7], properties {weights [1,1,1,1,1,1]}}
 {id=WavePlayer_0, properties {bLooping=true, soundWaveAssetPtr=&/Goop_Prj_InAir_Loop}}
 {id=WavePlayer_1, properties {soundWaveAssetPtr=&/Goop_Throw_Prj_Whoosh_01}}
 {id=WavePlayer_2, properties {soundWaveAssetPtr=&/Goop_Prj_InAir_Whip_Distant_01}}
 {id=WavePlayer_3, properties {soundWaveAssetPtr=&/Goop_Prj_InAir_Whip_Distant_02}}
 {id=WavePlayer_4, properties {soundWaveAssetPtr=&/Goop_Prj_InAir_Whip_Distant_03}}
 {id=WavePlayer_5, properties {soundWaveAssetPtr=&/Goop_Prj_InAir_Whip_Distant_04}}
 {id=WavePlayer_6, properties {soundWaveAssetPtr=&/Goop_Prj_InAir_Whip_Distant_05}}
 {id=WavePlayer_7, properties {soundWaveAssetPtr=&/Goop_Prj_InAir_Whip_Distant_06}}
]
properties {
 attenuationSettings=/Game/Sounds/Attenuation/Fort_Weapon_Launcher_Proj_InAir_Default_Att
 bOverrideConcurrency=true
 maxDistance=12750
 soundClassObject=&Fort_WeaponRanged_Fire_3P_PSM
 volumeMultiplier=0.8
 attenuationOverrides {falloffDistance=5000, attenuationShapeExtents {x=1000}}
 concurrencyOverrides {
  bVolumeScaleCanRelease=true
  maxCount=3
  resolutionRule=StopFarthestThenPreventNew
  volumeScale=0.2
  volumeScaleAttackTime=0.2
  volumeScaleReleaseTime=2
 }
 concurrencySet [/Game/Sounds/Concurrency_Rules/Athena_Explosive_Fuse_Conc]
}
