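# Sound cue for the Goop item held in hand, pairing an equip sound with a continuous effect loop.
# A mixer layers the one-shot equip wave with a looping FX wave, whose enveloper ramps volume up from silence over two seconds.
# Playback uses the weapon reload attenuation asset and the Fort_AGV sound class.
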SoundCue /MotherGameplay/Items/Goop/Audio/Goop_Prj_InHand_Loop_Cue
root=Mixer_0
nodes [
 {
  id=Enveloper_0
  children [WavePlayer_8]
  properties {pitchCurve {editorCurveData {keys [{time=0}]}}, volumeCurve {editorCurveData {keys [<at 0 0>,<at 2 1>]}}}
 }
 {id=Mixer_0, children [Enveloper_0,WavePlayer_0], properties {inputVolume [1,1]}}
 {id=WavePlayer_0, properties {soundWaveAssetPtr=&/Goop_Equip_01}}
 {id=WavePlayer_8, properties {bLooping=true, soundWaveAssetPtr=&/Goop_Equip_FX_Loop}}
]
properties {
 attenuationSettings=&Fort_Weapon_Reload_Att
 maxDistance=2250
 soundClassObject=/Game/Sounds/FortSoundClassesAndModes/Fort_AGV
 volumeMultiplier=1.12
}
