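# Sound cue for the close in-air whip-by sound of a Goop projectile, classed as a ranged weapon projectile effect.
# A Random node picks one of six equally weighted whip wave variations and feeds it through a Modulator left at default settings.
# It references the compound bow whoosh-by attenuation asset from Explosive Bow, overrides concurrency and has the debug flag on.
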SoundCue /MotherGameplay/Items/Goop/Audio/Goop_Prj_Whip_Close_Cue
root=Modulator_1
nodes [
 {id=Modulator_1, children [Random_1]}
 {id=Random_1, children [WavePlayer_0,WavePlayer_5,WavePlayer_6,WavePlayer_7,WavePlayer_8,WavePlayer_9], properties {weights [1,1,1,1,1,1]}}
 {id=WavePlayer_0, properties {soundWaveAssetPtr=&/Goop_Prj_InAir_Whip_Close_01}}
 {id=WavePlayer_5, properties {soundWaveAssetPtr=&/Goop_Prj_InAir_Whip_Close_02}}
 {id=WavePlayer_6, properties {soundWaveAssetPtr=&/Goop_Prj_InAir_Whip_Close_03}}
 {id=WavePlayer_7, properties {soundWaveAssetPtr=&/Goop_Prj_InAir_Whip_Close_04}}
 {id=WavePlayer_8, properties {soundWaveAssetPtr=&/Goop_Prj_InAir_Whip_Close_05}}
 {id=WavePlayer_9, properties {soundWaveAssetPtr=&/Goop_Prj_InAir_Whip_Close_06}}
]
properties {
 attenuationSettings=/Game/Sounds/Fort_Weapons/Explosive_Bow/Weapon_Compound_Bow_Whoosh_By_Att
 bDebug=true
 bOverrideConcurrency=true
 maxDistance=2097152
 soundClassObject=&Fort_WeaponRanged_Projectile_Effects
 volumeMultiplier=1
 attenuationOverrides {
  bAttenuate=false
  bEnableOcclusion=true
  bEnableReverbSend=false
  falloffDistance=0
  occlusionLowPassFilterFrequency=2000
  occlusionTraceChannel=ECC_GameTraceChannel11
  occlusionVolumeAttenuation=0
  spatializationAlgorithm=SPATIALIZATION_HRTF
  attenuationShapeExtents {x=50000}
 }
}
