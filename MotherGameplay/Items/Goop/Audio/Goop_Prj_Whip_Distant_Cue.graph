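# Sound cue for the distant in-air whip sound of the Goop projectile.
# A Random node picks one of six equally weighted whip waves and feeds a Modulator at the root.
# It uses the projectile effects sound class for ranged weapons, borrows the compound bow whoosh-by attenuation asset, and overrides concurrency.
# The cue's debug flag is left on.
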SoundCue /MotherGameplay/Items/Goop/Audio/Goop_Prj_Whip_Distant_Cue
root=Modulator_0
nodes [
 {id=Modulator_0, children [Random_0]}
 {id=Random_0, children [WavePlayer_1,WavePlayer_2,WavePlayer_3,WavePlayer_4,WavePlayer_8,WavePlayer_9], properties {weights [1,1,1,1,1,1]}}
 {id=WavePlayer_1, properties {soundWaveAssetPtr=&/Goop_Prj_InAir_Whip_Distant_01}}
 {id=WavePlayer_2, properties {soundWaveAssetPtr=&/Goop_Prj_InAir_Whip_Distant_02}}
 {id=WavePlayer_3, properties {soundWaveAssetPtr=&/Goop_Prj_InAir_Whip_Distant_03}}
 {id=WavePlayer_4, properties {soundWaveAssetPtr=&/Goop_Prj_InAir_Whip_Distant_04}}
 {id=WavePlayer_8, properties {soundWaveAssetPtr=&/Goop_Prj_InAir_Whip_Distant_05}}
 {id=WavePlayer_9, properties {soundWaveAssetPtr=&/Goop_Prj_InAir_Whip_Distant_06}}
]
properties {
 attenuationSettings=/Game/Sounds/Fort_Weapons/Explosive_Bow/Weapon_Compound_Bow_Whoosh_By_Att
 bDebug=true
 bOverrideConcurrency=true
 maxDistance=2097152
 soundClassObject=&Fort_WeaponRanged_Projectile_Effects
 volumeMultiplier=1
 attenuationOverrides {
  bAttenuate=false
  bEnableOcclusion=true
  bEnableReverbSend=false
  bUseComplexCollisionForOcclusion=true
  falloffDistance=1000
  occlusionTraceChannel=ECC_GameTraceChannel11
  occlusionVolumeAttenuation=0
  attenuationShapeExtents {x=50000}
 }
}
