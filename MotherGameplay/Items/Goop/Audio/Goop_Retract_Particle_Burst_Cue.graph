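# Sound cue for the particle burst heard when the Goop item retracts.
# A Random node picks among three equally weighted Goop close particle burst waves, passing through a Modulator into a Mixer whose first input is left unconnected.
# The cue plays at reduced volume in the Fort_AGV sound class, using the Goop activate center burst attenuation asset.
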SoundCue /MotherGameplay/Items/Goop/Audio/Goop_Retract_Particle_Burst_Cue
root=Mixer_1
nodes [
 {id=Mixer_1, children [null,Modulator_0], properties {inputVolume [2,1]}}
 {id=Modulator_0, children [Random_0]}
 {id=Random_0, children [WavePlayer_0,WavePlayer_1,WavePlayer_2], properties {weights [1,1,1]}}
 {id=WavePlayer_0, properties {soundWaveAssetPtr=&/Goop_Close_Particle_Burst_01}}
 {id=WavePlayer_1, properties {soundWaveAssetPtr=&/Goop_Close_Particle_Burst_02}}
 {id=WavePlayer_2, properties {soundWaveAssetPtr=&/Goop_Close_Particle_Burst_03}}
]
properties {
 attenuationSettings=&/Fort_Goop_Activate_Center_Burst_Att
 maxDistance=11000
 soundClassObject=/Game/Sounds/FortSoundClassesAndModes/Fort_AGV
 volumeMultiplier=0.675
}
