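# Sound cue for the Goop item's retract tell, presumably the warning that plays as the goop retracts.
# A mixer layers two branches, each picking one of three equally weighted random waves: quad "Inside" variants and "Outside" variants.
# Each branch has its own attenuation, reusing the Fort_Goop_Activate inside and outside settings, so the two perspectives are spatialized separately.
# The cue plays slightly boosted through the Fort_AGV sound class.
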SoundCue /MotherGameplay/Items/Goop/Audio/Goop_Retract_Tell_Cue
root=Mixer_0
nodes [
 {id=Attenuation_0, children [Random_3], properties {attenuationSettings=&/Fort_Goop_Activate_Inside_Att}}
 {id=Attenuation_1, children [Random_2], properties {attenuationSettings=&/Fort_Goop_Activate_Outside_Att}}
 {id=Mixer_0, children [Attenuation_0, Attenuation_1], properties {inputVolume [1,1]}}
 {id=Random_2, children [WavePlayer_6,WavePlayer_7,WavePlayer_8], properties {weights [1,1,1]}}
 {id=Random_3, children [WavePlayer_9, WavePlayer_10, WavePlayer_11], properties {weights [1,1,1]}}
 {id=WavePlayer_10, properties {soundWaveAssetPtr=&/Goop_Retract_Tell_Inside_Quad_02}}
 {id=WavePlayer_11, properties {soundWaveAssetPtr=&/Goop_Retract_Tell_Inside_Quad_03}}
 {id=WavePlayer_6, properties {soundWaveAssetPtr=&/Goop_Retract_Tell_Outside_01}}
 {id=WavePlayer_7, properties {soundWaveAssetPtr=&/Goop_Retract_Tell_Outside_02}}
 {id=WavePlayer_8, properties {soundWaveAssetPtr=&/Goop_Retract_Tell_Outside_03}}
 {id=WavePlayer_9, properties {soundWaveAssetPtr=&/Goop_Retract_Tell_Inside_Quad_01}}
]
properties {maxDistance=12000, soundClassObject=/Game/Sounds/FortSoundClassesAndModes/Fort_AGV, volumeMultiplier=1.25}
