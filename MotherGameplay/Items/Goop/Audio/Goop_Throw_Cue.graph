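# Sound cue for throwing the Goop item, playing the Goop_Throw_01 wave.
# A single wave player feeds a Modulator left at default settings.
# The cue uses the weapon reload attenuation asset and the Fort_AGV sound class, with a raised volume multiplier.
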SoundCue /MotherGameplay/Items/Goop/Audio/Goop_Throw_Cue
root=Modulator_0
nodes [{id=Modulator_0, children [WavePlayer_0]}, {id=WavePlayer_0, properties {soundWaveAssetPtr=&/Goop_Throw_01}}]
properties {
 attenuationSettings=&Fort_Weapon_Reload_Att
 maxDistance=2250
 soundClassObject=/Game/Sounds/FortSoundClassesAndModes/Fort_AGV
 volumeMultiplier=1.68
}
