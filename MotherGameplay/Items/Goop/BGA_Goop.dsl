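; Goop biome volume that flags pawns inside as in goop, ending skydiving and vertical velocity for parachuting ones.
; After a delay, the server kills it if it overlaps DudeBro Mini vent meshes, otherwise kills nearby deployed goop; both checks require the vent self-destruct setting.
; A replicated intersection struct triggers cosmetic enter/exit effects through OnRep.
; Clients scale volume FX to the collision mesh and spawn a matching audio volume.
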(event EventBeginPlay
  (|CallParentFunction)
  (Variables|Default|SetbShouldTestForOtherBiomes true)
  (Networking|SwitchHasAuthority
    (:Authority
      ;[1 0 0] Hack to mitigate stacking GC issue when deleting volumes players are inside
      (Utilities|FlowControl|Delay (Variables|Default|GetDelayDeletingOverlappingGoopVolumes))
      (Utilities|IsValid self
        (:Is%20Valid
          ;[1 1 1] Destroy overlapping Deployed Biomes if within min distance between biomes
          (Variables|SetBox%20Extent_X (Math|Vector|MakeVector (ScalableFloat|GetValueatLevel (Variables|Default|GetRow_MinXYDistanceBetweenBiomes)) (ScalableFloat|GetValueatLevel (Variables|Default|GetRow_MinXYDistanceBetweenBiomes)) (ScalableFloat|GetValueatLevel (Variables|Default|GetRow_MinZDistanceBetweenBiomes))))
          ;[1 0 0] Hack - Destroy self if overlapping vent actors (stuff that puts players into skydiving mode)
          (if (ScalableFloat|GetValueasBool (Variables|Default|GetRow_bSelfDestructIfOverlappingVent))
            (bind _outcomponents (Collision|BoxOverlapComponents (Transformation|GetActorLocation self) (Variables|Default|GetBox%20Extent) (Variables|Default|GetObject%20Types) "/Script/Engine.StaticMeshComponent" (Utilities|Array|MakeArray self) :!class /Script/Engine.KismetSystemLibrary))
            (Variables|Default|SetOverlappedStaticMeshComponents _outcomponents)
            (bind _outactors (Collision|BoxOverlapActors (Transformation|GetActorLocation self) (Variables|Default|GetBox%20Extent) (Variables|Default|GetObject%20Types) "/Script/FortniteGame.BuildingGameplayActor" (Utilities|Array|MakeArray self) :!class /Script/Engine.KismetSystemLibrary))
            (bind _array_element (Utilities|Array|ForEachLoopwithBreak _outactors :!label _label_2)
              (:LoopBody
                (if (Utilities|ClassIsChildOf (Utilities|GetClass _array_element) "/Game/Athena/Environments/Blueprints/DudeBro/BGA_DudeBro_Mini.BGA_DudeBro_Mini_C")
                  (bind _array_element_2 (Utilities|Array|ForEachLoopwithBreak (Actor|GetComponentsByClass _array_element "/Script/Engine.StaticMeshComponent") :!label _label)
                    (:LoopBody
                      (if (Utilities|Array|ContainsItem (Variables|Default|GetOverlappedStaticMeshComponents) _array_element_2)
                        (Variables|Default|SetbShouldTestForOtherBiomes false)
                        (CallFunction|KillSelf)
                        (Utilities|FlowControl|Sequence
                          (:then_0
                            (goto _label Break))
                          (:then_1
                            (goto _label_2 Break))))))))
              (:Completed
                (if (Variables|Default|GetbShouldTestForOtherBiomes)
                  (bind (_outactors_2 _returnvalue) (Collision|BoxOverlapActors (Transformation|GetActorLocation self) (Variables|Default|GetBox%20Extent) (Variables|Default|GetObject%20Types) (Variables|Default|GetActor%20Class%20Filter) (Utilities|Array|MakeArray self) :!class /Script/Engine.KismetSystemLibrary))
                  (if _returnvalue
                    (for _array_element_3 _outactors_2
                      (bind (_returnvalue_x _returnvalue_y _returnvalue_z) (Math|Vector|BreakVector (- (Variables|Default|GetBox%20Extent) (Math|Vector|VectorGetAbs (- (Transformation|GetActorLocation _array_element_3) (Transformation|GetActorLocation self))))))
                      (if (Math|Boolean|ANDBoolean (> _returnvalue_x 0.0) (> _returnvalue_y 0.0) (> _returnvalue_z 0.0))
                        (bind _asbga_goop_deployed (Utilities|Casting|CastToBGA_Goop_Deployed _array_element_3))
                        (CallFunction|KillSelf _asbga_goop_deployed)))))))))))
    (:Remote
      ;[1 1 1] Set scale of the goop volume fx
      (Effects|Components|ParticleSystem|SetVectorParameter (Variables|Default|GetNS_Biome_Volume_InWorld) "BoxSize" (Transformation|GetWorldScale (Variables|Default|GetCollisionMesh)))
      (CallFunction|UpdateAudioShapeTransform false))))

(event EventEndPlay
  (params (EndPlayReason :type EEndPlayReason%20Enum :default "Destroyed"))
  (Development|PrintString (Utilities|String|Append "Ended Play " (Utilities|String|EnumtoString EndPlayReason)) true true "(R=1.000000,G=0.000000,B=0.006524,A=1.000000)" 1e+01))

(fn ApplyPawnEffects
  (params (Pawn :type Fort%20Pawn%20Object%20Reference))
  (|CallParentFunction Pawn)
  (Variables|Default|SetPawnIntersectWorldLocation (Utilities|Struct|MakeBGoopIntersection (CallFunction|GetPawnIntersectLocation (Variables|GetPawn)) false (Variables|GetPawn)))
  (if (GameplayTags|HasAnyMatchingGameplayTags (Variables|GetPawn) (Variables|Default|GetTags_Parachuting))
    (bind _asfort_player_pawn (Utilities|Casting|CastToFortPlayerPawn (Variables|GetPawn))
      (:then
        (Parachute|EndSkydiving _asfort_player_pawn)
        (bind (_returnvalue_x _returnvalue_y) (Math|Vector|BreakVector (Pawn|Components|CharacterMovement|GetLastUpdateVelocity (Variables|/Script/Engine.Character|GetCharacterMovement _asfort_player_pawn))))
        (Variables|/Script/Engine.MovementComponent|SetVelocity_X (Math|Vector|MakeVector _returnvalue_x _returnvalue_y) (Variables|/Script/Engine.Character|GetCharacterMovement _asfort_player_pawn)))
      (:CastFailed
        _)))
  (Goop|SetIsinGoop (Variables|GetPawn) true))

(fn GetPawnIntersectLocation
  (params (Target :type Character%20Object%20Reference))
  (return (Math|Vector|vector+vector (Transformation|GetWorldLocation (Variables|/Script/Engine.Character|GetMesh Target)) (Math|Vector|MakeVector 0.0 0.0 (* (Components|Capsule|GetScaledCapsuleHalfHeight (Variables|/Script/Engine.Character|GetCapsuleComponent Target)) 1.25)))))

(fn OnRep_PawnIntersectWorldLocation
  (bind (_pawnintersectworldlocation_2_a2407e2f42c1b0775081668ef964681b _isexitingvolume_5_ae6b0e0d46f6db0cd472efbf83b4f1f9 _pawn_10_6f751f054dcd490a20f881bdab3d8dce) (Utilities|Struct|BreakBGoopIntersection (Variables|Default|GetPawnIntersectWorldLocation)))
  (Networking|CanExecuteCosmeticEvents
    (:True
      (CallFunction|PlayIntersectEffect _pawnintersectworldlocation_2_a2407e2f42c1b0775081668ef964681b _isexitingvolume_5_ae6b0e0d46f6db0cd472efbf83b4f1f9 _pawn_10_6f751f054dcd490a20f881bdab3d8dce))))

(fn RemovePawnEffects
  (params (Pawn :type Fort%20Pawn%20Object%20Reference))
  (|CallParentFunction Pawn)
  (Goop|SetIsinGoop (Variables|GetPawn) false)
  (Variables|Default|SetPawnIntersectWorldLocation (Utilities|Struct|MakeBGoopIntersection (CallFunction|GetPawnIntersectLocation (Variables|GetPawn)) true (Variables|GetPawn))))

(fn UpdateAudioShapeTransform
  (params (IsDynamicUpdate :type Boolean))
  (if (Math|Boolean|Equal(Boolean) IsDynamicUpdate (Variables|Default|GetIsDynamicTransform))
    (Variables|Default|GetAGVActor
      (:then)
      (:else
        (bind _returnvalue (Game|SpawnActorfromClass "/MotherGameplay/Items/Goop/Audio/AGV_Actor_Goop.AGV_Actor_Goop_C" (Transformation|GetWorldTransform (Variables|Default|GetCollisionMesh)) "Undefined" "OverrideRootScale" 0 0))
        (Variables|Default|SetAGVActor _returnvalue)))))

(fn ConstructionScript
  (|CallParentFunction)
  (Networking|CanExecuteCosmeticEvents
    (:True
      ;[1 1 1] Set scale of the goop volume fx
      (Effects|Components|ParticleSystem|SetVectorParameter (Variables|Default|GetNS_Biome_Volume_InWorld) "BoxSize" (Transformation|GetWorldScale (Variables|Default|GetCollisionMesh))))))
