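; Deployed Goop volume actor that sizes and offsets its collision mesh from scalable-float data and dies on a kill-volume event or other death.
; The server ensures the instigator has a tracker effect, sending add and remove volume events on spawn and death.
; Replicated spawn time schedules expiry warning and transition-out from a duration curve.
; Custom primitive data drives material transitions, alongside Niagara and sound effects; late-arriving clients skip the spawn transition.
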(event Custom|AddTracker
  (bind _instigator (Variables|Actor|GetInstigator))
  (bind _returnvalue (Ability|GetAbilitySystemComponent _instigator))
  ;[1 1 1] Add Tracker
  (Networking|SwitchHasAuthority
    (:Authority
      (Utilities|IsValid (Variables|Actor|GetInstigator)
        (:Is%20Valid
          (Utilities|IsValid _returnvalue
            (:Is%20Valid
              (if (not (> (GameplayEffects|GetGameplayEffectCount _returnvalue (Variables|Default|GetGE_Tracker) 0 true) 0))
                (Ability|Tasks|WaitGameplayTagAddToActor (Variables|Actor|GetInstigator) (Variables|Default|GetTag_GoopTracker) true
                  (:then
                    (GameplayEffects|ApplyGameplayEffectToSelf (Ability|GetAbilitySystemComponent (Variables|Actor|GetInstigator)) (Variables|Default|GetGE_Tracker) 0.0 (Utilities|Struct|MakeGameplayEffectContextHandle)))
                  (:Added
                    _)) ; This tag is granted by the tracker ability granted by the GE
              )
              (Ability|SendGameplayEventtoActor (Variables|Actor|GetInstigator) (Variables|Default|GetEventTag_GoopTracker_AddVolume) (Utilities|Struct|MakeGameplayEventData "(TagName=\"\")" 0 self)))))))))

(event EventOnDeathServer
  (params (Damage :type Float%20%28single-precision%29) (DamageTags :type Gameplay%20Tag%20Container%20Structure%20%28by%20ref%29) (Momentum :type Vector :default "0, 0, 0") (HitInfo :type Hit%20Result%20Structure%20%28by%20ref%29) (InstigatedBy :type Controller%20Object%20Reference) (DamageCauser :type Actor%20Object%20Reference) (EffectContext :type Gameplay%20Effect%20Context%20Handle%20Structure))
  ;[1 1 1] On Death (Server Only)
  (|CallParentFunction Damage DamageTags Momentum HitInfo InstigatedBy DamageCauser EffectContext)
  (Utilities|IsValid (Variables|Actor|GetInstigator)
    (:Is%20Valid
      (Ability|SendGameplayEventtoActor (Variables|Actor|GetInstigator) (Variables|Default|GetEventTag_GoopTracker_RemoveVolume) (Utilities|Struct|MakeGameplayEventData "(TagName=\"\")" 0 self)))))

(event Custom|PlaySpawnFX
  (bind _defaultsceneroot (Variables|Default|GetDefaultSceneRoot))
  (bind _returnvalue (Transformation|GetWorldLocation _defaultsceneroot))
  ;[1 1 1] Play Spawn FX
  (Niagara|SpawnSystematLocation :SystemTemplate "/MotherGameplay/FX/Biome/NS_Biome_Volume_Spawn.NS_Biome_Volume_Spawn" :Location _returnvalue :Scale "1.000000,1.000000,1.000000" :bAutoDestroy true :bAutoActivate true :PoolingMethod "AutoRelease" :bPreCullCheck true)
  (Audio|PlaySoundatLocation "/MotherGameplay/Items/Goop/Audio/Goop_Expand_Cue.Goop_Expand_Cue" _returnvalue)
  (CallFunction|TransitionGoopIn (Variables|Default|GetTransitionInLength))
  (Development|SetHiddeninGame (Variables|Default|GetNS_Biome_Volume_InWorld) false false))

(event EventBeginPlay
  (bind (_returnvalue_location _returnvalue_rotation) (Math|Transform|BreakTransform (Transformation|GetWorldTransform (Variables|Default|GetCollisionMesh))))
  (bind (_returnvalue_location_x _returnvalue_location_y _returnvalue_location_z) (Math|Vector|BreakVector _returnvalue_location))
  ;[1 1 1] Begin Play - set custom transform from gameplay data
  (|CallParentFunction)
  (CallFunction|AddTracker)
  (Ability|Async|WaitGameplayEventToActor self (Variables|Default|GetEventTag_Goop_KillVolume)
    (:then
      (Transformation|SetWorldTransform (Variables|Default|GetCollisionMesh) (Math|Transform|MakeTransform (Math|Vector|MakeVector _returnvalue_location_x _returnvalue_location_y (Math|Float|float+float _returnvalue_location_z (ScalableFloat|GetValueatLevel (Variables|Default|GetRow_SpawnOffset_Z)) (Variables|Default|GetMeshZBias))) _returnvalue_rotation (Math|Vector|MakeVector (ScalableFloat|GetValueatLevel (Variables|Default|GetRow_Size_X)) (ScalableFloat|GetValueatLevel (Variables|Default|GetRow_Size_Y)) (ScalableFloat|GetValueatLevel (Variables|Default|GetRow_Size_Z)))) false true)
      (Networking|SwitchHasAuthority
        (:Authority
          (CallFunction|PlaySpawnFX)
          (Variables|Default|SetSpawnTime (Fortnite|Sync|Time|GetServerWorldTimeSeconds)))
        (:Remote
          (CallFunction|UpdateAudioShapeTransform true)
          (CallFunction|TransitionGoopIn 0.0) ; Do this instantly when not the first spawn
          ;[1 1 1] Set scale on the Niagara particle volume
          (bind _ns_biome_volume_inworld (Variables|Default|GetNS_Biome_Volume_InWorld)
            (:then
              (Effects|Components|ParticleSystem|SetVectorParameter _ns_biome_volume_inworld "BoxSize" (Transformation|GetWorldScale (Variables|Default|GetCollisionMesh))))
            (:else
              _))
          (Utilities|FlowControl|Delay (Variables|Default|GetDelayParticleEffectAfterFirstSpawn)) ; Delay to avoid showing the "shoot out" animation on particules
          (Development|SetHiddeninGame (Variables|Default|GetNS_Biome_Volume_InWorld) false false))))
    (:EventReceived
      (CallFunction|KillSelf))))

(event Custom|SetExpireWarningAndTransitionOutTimers
  (bind _returnvalue (Fortnite|Sync|Time|GetServerWorldTimeSeconds))
  (bind _spawntime (Variables|Default|GetSpawnTime))
  (bind _returnvalue_2 (- _returnvalue _spawntime))
  ;[1 1 1] Set Expire Warning and Transition Out Timers
  (bind _outxy (DataTable|EvaluateCurveTableRow "/MotherGameplay/DataTables/MotherGameData.MotherGameData" "Default.Mother.Goop.AoE.Duration")
    (:RowFound
      (Utilities|Time|SetTimerbyEvent (EventDispatchers|CreateEvent :!event TransitionGoopOut) (Utilities|ClampMinFloat (- (- _outxy _returnvalue_2) (Variables|Default|GetTransitionOutLength)) 0.1))
      (Utilities|Time|SetTimerbyEvent (EventDispatchers|CreateEvent :!event ActivateVisualsExpireWarning) (Utilities|ClampMinFloat (- (- _outxy _returnvalue_2) (ScalableFloat|GetValueatLevel (Variables|Default|GetRow_VolumeExpirationWarningDuration))) 0.1)))))

(event EventKillSelf
  (bind _defaultsceneroot (Variables|Default|GetDefaultSceneRoot))
  (bind _returnvalue (Transformation|GetWorldLocation _defaultsceneroot))
  ;[1 1 1] Spawn removal effects
  (|CallParentFunction)
  (CallFunction|ClientKillSelfCleanup)
  (Niagara|SpawnSystematLocation :SystemTemplate "/MotherGameplay/FX/Biome/NS_Biome_Volume_Remove.NS_Biome_Volume_Remove" :Location _returnvalue :Scale "1.000000,1.000000,1.000000" :bAutoDestroy true :bAutoActivate true :PoolingMethod "AutoRelease" :bPreCullCheck true)
  (Audio|PlaySoundatLocation "/MotherGameplay/Items/Goop/Audio/Goop_Retract_Particle_Burst_Cue.Goop_Retract_Particle_Burst_Cue" _returnvalue))

(event Custom|Client%20Kill%20Self%20Cleanup
  ;[1 1 1] Spawn removal effects
  (bind _agvactor (Variables|Default|GetAGVActor)
    (:then
      (Actor|DestroyActor _agvactor))
    (:else)))

(event EventEndPlay
  (params (EndPlayReason :type EEndPlayReason%20Enum :default "Destroyed"))
  ;[1 1 1] Destroy AGV actor on End Play
  (|CallParentFunction EndPlayReason)
  (bind _agvactor (Variables|Default|GetAGVActor)
    (:then
      (Actor|DestroyActor _agvactor))
    (:else)))

(fn ActivateVisualsExpireWarning
  (bind _collisionmesh (Variables|Default|GetCollisionMesh))
  (Rendering|Material|SetCustomPrimitiveDataFloat _collisionmesh 9 2.0)
  (Rendering|Material|SetCustomPrimitiveDataFloat _collisionmesh 10 (Math|Float|Fraction (Utilities|Time|GetGameTimeinSeconds)))
  (Audio|SpawnSoundAttached "/MotherGameplay/Items/Goop/Audio/Goop_Retract_Tell_Cue.Goop_Retract_Tell_Cue" (Variables|Default|GetDefaultSceneRoot)))

(fn OnRep_SpawnTime
  (Networking|CanExecuteCosmeticEvents
    (:True
      (CallFunction|SetExpireWarningandTransitionOutTimers))))

(fn TransitionGoopIn
  (params (TransitionInLength :type Float%20%28double-precision%29))
  (bind _collisionmesh (Variables|Default|GetCollisionMesh))
  (Rendering|Material|SetCustomPrimitiveDataFloat _collisionmesh 5 (Utilities|Time|GetGameTimeinSeconds))
  (Rendering|Material|SetCustomPrimitiveDataFloat _collisionmesh 6 1e+09)
  (Rendering|Material|SetCustomPrimitiveDataFloat _collisionmesh 7 TransitionInLength))

(fn TransitionGoopOut
  (bind _collisionmesh (Variables|Default|GetCollisionMesh))
  (Rendering|Material|SetCustomPrimitiveDataFloat _collisionmesh 5 0.0)
  (Rendering|Material|SetCustomPrimitiveDataFloat _collisionmesh 6 (Utilities|Time|GetGameTimeinSeconds))
  (Rendering|Material|SetCustomPrimitiveDataFloat _collisionmesh 8 (Variables|Default|GetTransitionOutLength))
  (Audio|PlaySoundatLocation "/MotherGameplay/Items/Goop/Audio/Goop_Retract_Cue.Goop_Retract_Cue" (Transformation|GetActorLocation)))

(fn ConstructionScript)
