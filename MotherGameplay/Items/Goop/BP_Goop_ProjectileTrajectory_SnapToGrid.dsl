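; Goop projectile trajectory Blueprint for snap-to-grid placement that sizes and animates the SM_BuildingPreview mesh.
; BeginPlay builds Preview and Spawn relative transforms from ScalableFloat sizes and Z offsets on the mesh's current transform, then calls the parent.
; If scaling preview is enabled, a timeline lerps from Preview to Spawn transform, replayed by a looping three-second timer.
; Otherwise, the mesh is set to the Preview transform.
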(event Custom|ScalePreview
  ;[1 1 1] Scale Preview
  (bind (_direction _scalar) (Timeline|PreviewScale PlayFromStart)
    (:Update
      (Transformation|SetRelativeTransform (Variables|Default|GetSM_BuildingPreview) (Math|Transform|Lerp(Transform) (Variables|Default|GetTransform_Preview) (Variables|Default|GetTransform_Spawn) _scalar) false true))))

(event EventBeginPlay
  (bind (_returnvalue_location _returnvalue_rotation) (Math|Transform|BreakTransform (Transformation|GetRelativeTransform (Variables|Default|GetSM_BuildingPreview))))
  (bind (_returnvalue_location_x _returnvalue_location_y _returnvalue_location_z) (Math|Vector|BreakVector _returnvalue_location))
  (bind (_returnvalue_location_2 _returnvalue_rotation_2) (Math|Transform|BreakTransform (Transformation|GetRelativeTransform (Variables|Default|GetSM_BuildingPreview))))
  (bind (_returnvalue_location_x_2 _returnvalue_location_y_2 _returnvalue_location_z_2) (Math|Vector|BreakVector _returnvalue_location_2))
  ;[1 1 1] Begin Play
  (Variables|Default|SetTransform_Preview (Math|Transform|MakeTransform (Math|Vector|MakeVector _returnvalue_location_x_2 _returnvalue_location_y_2 (+ _returnvalue_location_z_2 (ScalableFloat|GetValueatLevel (Variables|Default|GetRow_Preview_SpawnOffset_Z)))) _returnvalue_rotation_2 (Math|Vector|MakeVector (ScalableFloat|GetValueatLevel (Variables|Default|GetRow_Preview_Size_X)) (ScalableFloat|GetValueatLevel (Variables|Default|GetRow_Preview_Size_Y)) (ScalableFloat|GetValueatLevel (Variables|Default|GetRow_Preview_Size_Z)))))
  (Variables|Default|SetTransform_Spawn (Math|Transform|MakeTransform (Math|Vector|MakeVector _returnvalue_location_x _returnvalue_location_y (+ _returnvalue_location_z (ScalableFloat|GetValueatLevel (Variables|Default|GetRow_Spawn_SpawnOffset_Z)))) _returnvalue_rotation (Math|Vector|MakeVector (ScalableFloat|GetValueatLevel (Variables|Default|GetRow_Spawn_Size_X)) (ScalableFloat|GetValueatLevel (Variables|Default|GetRow_Spawn_Size_Y)) (ScalableFloat|GetValueatLevel (Variables|Default|GetRow_Spawn_Size_Z)))))
  (if (ScalableFloat|GetValueasBool (Variables|Default|GetRow_bUseScalingPreview))
    (CallFunction|ScalePreview)
    (Utilities|Time|SetTimerbyEvent (EventDispatchers|CreateEvent :!event Custom|ScalePreview) 3.0 true)
    (else
      (Transformation|SetRelativeTransform (Variables|Default|GetSM_BuildingPreview) (Variables|Default|GetTransform_Preview) false true)))
  (|CallParentFunction))

(fn ConstructionScript
  (|CallParentFunction))
