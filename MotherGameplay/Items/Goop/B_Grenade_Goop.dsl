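; Blueprint for the held Goop grenade item, adding idle audio and Niagara effects that follow the weapon's visibility.
; When the weapon becomes visible, it plays the in-hand loop sound and activates the idle biome projectile effect.
; When hidden, it fades the loop out with a linear curve and deactivates the effect, with a redundant second deactivate call.
; BeginPlay, Tick and the construction script add no logic of their own.
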(event EventBeginPlay
  (|CallParentFunction) ; This node is disabled and will not be called.\nDrag off pins to build functionality.
)

(event EventTick
  (params (DeltaSeconds :type Float%20%28single-precision%29))
  (|CallParentFunction DeltaSeconds) ; This node is disabled and will not be called.\nDrag off pins to build functionality.
)

(event EventOnWeaponVisibilityChanged
  (params (bVisible :type Boolean) (bSetForLocalControllerOnly :type Boolean))
  (bind _goop_inhand_loop (Variables|Default|GetGoop_InHand_Loop))
  (|CallParentFunction bVisible bSetForLocalControllerOnly)
  (if bVisible
    (Audio|Components|Audio|Play _goop_inhand_loop 0.0)
    (Components|Activation|Activate (Variables|Default|GetNS_Biome_Projectile_Idle) true)
    (else
      (Audio|Components|Audio|FadeOut _goop_inhand_loop 0.5 0.0 "Linear" :!class /Script/Engine.AudioComponent)
      (Components|Activation|Deactivate (Variables|Default|GetNS_Biome_Projectile_Idle))
      (Components|Activation|Deactivate (Variables|Default|GetNS_Biome_Projectile_Idle)))))

(fn ConstructionScript
  (|CallParentFunction))
