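# Additive, unlit, two-sided material for the deployable Goop Biome volume, rendering a glowing, shimmering energy shell.
# Fresnel, panning world-aligned noise, grain, reflection-based chroma and UV edge masks build the color, alongside camera-distance masks and near-camera fade.
# Custom primitive data drives timestamped deploy transitions, warning pulses and noise-distorted enter/exit ripples.
# World position offset adds transition deformation and intersection jiggle, while quality and shading-path switches simplify costly effects.
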Material /MotherGameplay/Items/Goop/M_GoopBiome
graph {
 properties {
  bEnableResponsiveAA=true
  blendMode=BLEND_Additive
  pixelDepthOffsetMode=PDOM_Legacy
  shadingModel=MSM_Unlit
  translucencyLightingMode=TLM_VolumetricPerVertexDirectional
  twoSided=true
  usedShadingModels=Unlit
 }
 nodes [
  n1 = Constant()
  n2 = Multiply(A=NamedRerouteUsage(), B=NamedRerouteUsage())
  n3 = LinearInterpolate(
   A=ScalarParameter() {defaultValue=0.15, parameterName=FresnelLow}
   Alpha=MaterialFunctionCall(Exponent=ScalarParameter() {defaultValue=2, parameterName=FresnelExponent}) {materialFunction=&MF_Fresnel_FX}.Fresnel Inverted Double Sided
  )
  n4 = ShadingPathSwitch(Default=Add(A=QualitySwitch(Default=@n1, Epic=@n2), B=@n3), Mobile=@n3) {desc=Migrated from FeatureLevelSwitch}
  n5 = Multiply(A=@n3, B=NamedRerouteUsage())
  n6 = LinearInterpolate(
   A=VectorParameter() {parameterName=ColorBaseA, defaultValue=<rgba 0.1184 0.0469 0.3038 0>}.RGB
   B=VectorParameter() {parameterName=ColorBaseB, defaultValue=<rgba 0.0718 0.2441 0.6163 0>}.RGB
   Alpha=NamedRerouteUsage()
  )
  n7 = VectorParameter() {parameterName=ColorSparkles, defaultValue=<rgba 0.0569 0.3042 0.8 0>}
  n8 = Add(A=NamedRerouteUsage(), B=Add(A=NamedRerouteUsage(), B=NamedRerouteUsage()))
  n9 = Add(A=NamedRerouteUsage(), B=NamedRerouteUsage())
  n10 = NamedRerouteUsage()
  n11 = LinearInterpolate(
   A=Multiply(A=Add(A=@n4, B=@n5), B=LinearInterpolate(A=@n6, B=@n7.RGB, Alpha=@n8))
   B=VectorParameter() {parameterName=ColorEdges, defaultValue=<rgba 0.5422 0.5422 25 0>}.RGB
   Alpha=StaticSwitchParameter(True=Add(A=@n9, B=@n10), False=@n9) {defaultValue=true, parameterName=UseTransition}
  )
  LinearInterpolate(A=@n11, B=Multiply(A=@n11, B=VertexColor().A), Alpha=NamedRerouteUsage()) {constA=1, constB=0}
  MaterialFunctionCall!(Time, Divisor=Constant() {r=2}, Mod) {materialFunction=/Game/Effects/Fort_Effects/Materials/_Functions/MF_ModuloTime}
  MaterialFunctionCall(Divisor=Constant() {r=2}) {materialFunction=&MF_ModuloTime}
  n12 = VertexNormalWS()
  n13 = Abs(None=@n12)
  n14 = MaterialFunctionCall() {materialFunction=&ObjectScale}
  n15 = Subtract(A=Multiply(A=CameraPositionWS(), B=@n13), B=Multiply(A=@n12, B=Divide(A=@n14.Scale XYZ)))
  n16 = Multiply(
   A=@n13
   B=MaterialFunctionCall() {materialFunction=&ObjectPivotPoint}.Object Pivot Location
  )
  n17 = ScalarParameter() {defaultValue=300, parameterName=DistanceDivisorA}
  n18 = ScalarParameter() {defaultValue=1200, parameterName=DistanceDivisorB}
  n19 = MaterialFunctionCall(Float3=VertexNormalWS()) {materialFunction=&BreakOutFloat3Components}
  n20 = Abs(None=@n19.B)
  n21 = MaterialFunctionCall(Float3=MaterialFunctionCall() {materialFunction=&CameraDirectionVector}.Result) {materialFunction=&BreakOutFloat3Components}
  n22 = Normalize(VectorInput=MaterialFunctionCall(X=@n21.R, Y=@n21.R, Z=Constant() {r=1}) {materialFunction=&MakeFloat3}.Result)
  n23 = Abs(None=DotProduct(
    A=MaterialFunctionCall(X=@n19.R, Y=@n19.R, Z=@n20) {materialFunction=&MakeFloat3}.Result
    B=Multiply(A=@n22) {constB=-1}
   ))
  n24 = Saturate(None=Divide(A=Distance(A=@n15, B=@n16), B=LinearInterpolate(A=@n17, B=@n18, Alpha=@n23)))
  NamedRerouteDeclaration(None=@n24) {name=DistanceCheck, nodeColor=<rgba 0.1183 0 1 1>}
  n25 = ScalarParameter() {defaultValue=0.3, parameterName=EdgeBaseSubtleScale}
  n26 = TextureCoordinate()
  n27 = MaterialFunctionCall(Float2=Abs(None=Subtract(A=@n26) {constB=0.5})) {materialFunction=&BreakOutFloat2Components}
  n28 = ScalarParameter() {defaultValue=0.004, parameterName=EdgeRadius}
  n29 = NamedRerouteUsage()
  n30 = Multiply(A=Subtract(A=@n29) {constB=0.25}, B=ScalarParameter() {defaultValue=0.008, parameterName=EdgeNoiseScale})
  n31 = SphereMask(
   A=Max(A=@n27.R, B=@n27.R)
   B=Constant() {r=0.5}
   Radius=Max(A=ScalarParameter() {defaultValue=0.002, parameterName=EdgeRadiusMin}, B=Add(A=@n28, B=@n30))
   Hardness=LinearInterpolate(B=ScalarParameter() {defaultValue=0.4, parameterName=EdgeHardness}, Alpha=@n29)
  )
  n32 = ScalarParameter() {defaultValue=8, parameterName=EdgeExponent}
  NamedRerouteDeclaration(None=Add(A=Multiply(A=@n25, B=@n31), B=Power(Base=@n31, Exp=@n32))) {name=Edges, nodeColor=<rgba 0.6641 0 1 1>}
  n33 = NamedRerouteUsage()
  n34 = NamedRerouteUsage()
  n35 = Constant()
  n36 = Transform(None=Subtract(A=MaterialFunctionCall(R=@n33, G=@n34, B=@n35) {materialFunction=&AppendMany}.RGB) {constB=0.5})
  n37 = ScalarParameter() {defaultValue=100, parameterName=IntersectionDistortionScale}
  n38 = MaterialFunctionCall(
   TimestampA=VectorParameter() {bUseCustomPrimitiveData=true, parameterName=IntersectionPosition}.A
   TransitionLength_In=ScalarParameter() {defaultValue=0.6, parameterName=Intersect_TransitionLength_In}
   Transition In Only?=StaticBool() {value=true}
  ) {materialFunction=&MF_Timestamp_InOut_01}
  n39 = VertexInterpolator(VS=@n38.Linear)
  n40 = LinearInterpolate(
   A=@n39.PS
   B=OneMinus(None=@n39.PS)
   Alpha=ScalarParameter() {bUseCustomPrimitiveData=true, parameterName=ExitingVolume, primitiveDataIndex=4}
  )
  n41 = SphereMask(
   A=VectorParameter() {bUseCustomPrimitiveData=true, parameterName=IntersectionPosition}.RGB
   B=Add(A=Multiply(A=@n36, B=@n37), B=WorldPosition().XYZ)
   Radius=Multiply(A=ScalarParameter() {defaultValue=700, parameterName=IntersectRadius}, B=LinearInterpolate(Alpha=@n40) {constA=0.2})
   Hardness=Constant()
  )
  n42 = Constant() {r=0.5}
  n43 = NamedRerouteDeclaration(None=OneMinus(None=@n38.Linear)) {name=TimestampInverted, nodeColor=<rgba 1 0.2812 0 1>}
  n44 = Multiply(A=@n42, B=@n43)
  n45 = Constant()
  n46 = Multiply(A=Power(Base=@n43) {constExponent=8}, B=ScalarParameter() {defaultValue=40, parameterName=PulseIntensity})
  n47 = NamedRerouteUsage()
  n48 = Add(
   A=Multiply(A=SphereMask(A=@n41, B=@n42, Radius=@n44, Hardness=@n45), B=ScalarParameter() {defaultValue=40, parameterName=IntersectionScale})
   B=Add(A=Multiply(A=@n46, B=@n47), B=ScalarParameter() {defaultValue=2, parameterName=PulseBias})
  )
  NamedRerouteDeclaration(None=QualitySwitch(Default=Constant(), Epic=Multiply(A=@n48, B=@n43))) {name=SphereMaskIntersection, nodeColor=<rgba 0 1 0.6719 1>}
  n49 = Power(Base=MaterialFunctionCall!(Distance Bias, Distance Divisor=ScalarParameter() {defaultValue=1, parameterName=IntersectEdgeDivisor}, Scene Depth, Pixel Depth) {materialFunction=/Game/Effects/Fort_Effects/Materials/_Functions/MF_DistanceBehindTranslucency_01}.Normalized Inverted) {constExponent=3}
  n50 = MaterialFunctionCall(Float3=VertexNormalWS()) {materialFunction=&BreakOutFloat3Components}
  n51 = DotProduct(
   A=MaterialFunctionCall(X=Abs(None=@n50.R), Y=Abs(None=@n50.G), Z=Saturate(None=@n50.B)) {materialFunction=&MakeFloat3}.Result
   B=Constant3Vector() {constant=<rgba 1 1 0 0>}
  )
  NamedRerouteDeclaration(None=QualitySwitch(Default=Constant(), Epic=Multiply(A=@n49, B=@n51))) {name=IntersectionEdges, nodeColor=<rgba 1 0.1172 0 1>}
  NamedRerouteDeclaration(None=VertexInterpolator(VS=OneMinus(None=@n24)).PS) {name=DistanceCheckInvertedVertexInterpolator, nodeColor=<rgba 1 0 0.7266 1>}
  n52 = MaterialFunctionCall() {materialFunction=&BoundingBoxBased_0-1_UVW}
  n53 = Constant()
  n54 = ScalarParameter() {bUseCustomPrimitiveData=true, parameterName=DeployTransitionTimestampA, primitiveDataIndex=5}
  n55 = ScalarParameter() {bUseCustomPrimitiveData=true, defaultValue=0.5, parameterName=DeployTransitionLength_In, primitiveDataIndex=7}
  n56 = ScalarParameter() {bUseCustomPrimitiveData=true, parameterName=DeployTransitionTimestampB, primitiveDataIndex=6}
  n57 = ScalarParameter() {bUseCustomPrimitiveData=true, parameterName=DeployTransitionLength_Out, primitiveDataIndex=8}
  n58 = StaticBool()
  n59 = NamedRerouteDeclaration(None=ConstantBiasScale(None=MaterialFunctionCall(TimestampA=@n54, TransitionLength_In=@n55, TimestampB=@n56, TransitionLength_Out=@n57, Transition In Only?=@n58) {materialFunction=&MF_Timestamp_InOut_01}.SmoothStep) {bias=-0.1, scale=1.2}) {name=TransitionTimestamp, nodeColor=<rgba 0 1 0.8594 1>}
  n60 = LinearInterpolate(A=ScalarParameter() {defaultValue=0.8, parameterName=TransitionHardness}, Alpha=@n59)
  n61 = Constant() {r=0.5}
  n62 = SphereMask(
   A=NamedRerouteDeclaration(None=SphereMask(A=@n52.B, B=@n53, Radius=@n59, Hardness=@n60)) {name=Transition, nodeColor=<rgba 0 0.6016 1 1>}
   B=@n61
   Radius=@n61
   Hardness=Constant()
  )
  NamedRerouteDeclaration(None=Multiply(A=Power(Base=@n62) {constExponent=4}, B=ScalarParameter() {defaultValue=200, parameterName=TransitionEdgeScale})) {name=TransitionEdge, nodeColor=<rgba 0.0781 1 0 1>}
  n63 = Sine(None=Subtract(A=Time(), B=ScalarParameter() {bUseCustomPrimitiveData=true, parameterName=WarningTellTimeBias, primitiveDataIndex=10})) {period=0.5}
  NamedRerouteDeclaration(None=Multiply(
    A=ConstantBiasScale(None=@n63) {bias=0.5}
    B=ScalarParameter() {bUseCustomPrimitiveData=true, parameterName=WarningTell, primitiveDataIndex=9}
   )) {name=WarningPulse, nodeColor=<rgba 0.6406 0 1 1>}
  n64 = Multiply(
   A=Subtract(
    A=MaterialFunctionCall() {materialFunction=&ObjectPivotPoint}.Object Pivot Location
    B=WorldPosition().XYZ
   )
   B=Constant3Vector() {constant=<rgba 1 1 0 0>}
  )
  n65 = Subtract(
   A=MaterialFunctionCall() {materialFunction=&ObjectPivotPoint}.Object Pivot Location
   B=WorldPosition().XYZ
  )
  n66 = Multiply(A=Constant3Vector() {constant=<rgba 1 1 0 0>}, B=@n65)
  n67 = Time()
  n68 = Multiply(
   A=Multiply(A=Normalize(VectorInput=@n66), B=Sine(None=@n67) {period=0.1})
   B=ScalarParameter() {defaultValue=10, parameterName=IntersectJiggleScale}
  )
  n69 = NamedRerouteUsage()
  n70 = OneMinus(None=Ceil(None=VertexColor().A))
  n71 = NamedRerouteUsage()
  n72 = Add(
   A=QualitySwitch(Default=Constant(), Epic=Multiply(A=@n68, B=@n69))
   B=Multiply(A=Multiply(A=@n65, B=@n70), B=Truncate!(None=@n71) {is=class MaterialExpressionTruncate})
  )
  n73 = NamedRerouteUsage()
  n74 = Saturate(None=SmoothStep(Value=Multiply(A=@n73) {constB=2}))
  NamedRerouteDeclaration(None=StaticSwitchParameter(True=LinearInterpolate(A=@n64, B=@n72, Alpha=@n74), False=@n72) {defaultValue=true, parameterName=UseTransition}) {name=WPO, nodeColor=<rgba 0 0.2734 1 1>}
  NamedRerouteDeclaration(None=VertexInterpolator(VS=@n24).PS) {name=DistanceCheckVertexInterpolator, nodeColor=<rgba 0 1 0.3906 1>}
  n75 = WorldPosition()
  n76 = ComponentMask(None=@n75.XYZ) {b=true, r=true}
  n77 = ComponentMask(None=@n75.XYZ) {b=true, g=true}
  n78 = VertexColor()
  NamedRerouteDeclaration(None=LinearInterpolate(A=LinearInterpolate(A=@n76, B=@n77, Alpha=@n78.R), B=ComponentMask(None=@n75.XYZ) {g=true, r=true}, Alpha=@n78.B)) {name=CheapWorldAlignedUVs, nodeColor=<rgba 0 0.7656 1 1>}
  n79 = Divide(A=NamedRerouteUsage(), B=ScalarParameter() {defaultValue=1000, parameterName=WorldTileDivisorAB})
  n80 = VectorParameter() {parameterName=NoiseA_TilePan, defaultValue=<rgba 2 2 0.0225 0.0194>}
  n81 = Panner(Coordinate=Multiply(A=@n79, B=AppendVector(A=@n80.R, B=@n80.R)), Speed=AppendVector(A=@n80.B, B=@n80.B)) {bFractionalPart=true}
  n82 = VectorParameter() {parameterName=NoiseB_TilePan, defaultValue=<rgba -2 -2 0.014 0.0129>}
  n83 = Panner(Coordinate=Multiply(A=@n79, B=AppendVector(A=@n82.R, B=@n82.R)), Speed=AppendVector(A=@n82.B, B=@n82.B)) {bFractionalPart=true}
  n84 = Multiply(
   A=NamedRerouteUsage()
   B=Add(
    A=TextureSample(UVs=@n81) {texture=/Game/Global/Textures/Default/PointSplatter/T_Tiling_Dust}.R
    B=TextureSample(UVs=@n83) {texture=&T_Tiling_Dust}.R
   )
  )
  n85 = Multiply(A=@n84, B=ScalarParameter() {defaultValue=4.5, parameterName=NoiseBrightScale})
  n86 = ScalarParameter() {defaultValue=5.5, parameterName=NoiseExponent}
  n87 = Divide(A=NamedRerouteUsage(), B=ScalarParameter() {defaultValue=3500, parameterName=WorldTileDivisorC})
  n88 = VectorParameter() {parameterName=NoiseC_TilePan, defaultValue=<rgba 2 2 0.0875 -0.077>}
  n89 = AppendVector(A=@n88.R, B=@n88.R)
  n90 = TextureSample(UVs=Panner(Coordinate=Multiply(A=@n87, B=@n89), Speed=AppendVector(A=@n88.B, B=@n88.B)) {bFractionalPart=true}) {texture=&T_ShieldCaustics}
  n91 = Divide(A=NamedRerouteUsage(), B=ScalarParameter() {defaultValue=3500, parameterName=WorldTileDivisorC})
  n92 = VectorParameter() {parameterName=NoiseD_TilePan, defaultValue=<rgba -2 -2 0.075 0.0605>}
  n93 = AppendVector(A=@n92.R, B=@n92.R)
  n94 = TextureSample(UVs=Panner(Coordinate=Multiply(A=@n91, B=@n93), Speed=AppendVector(A=@n92.B, B=@n92.B)) {bFractionalPart=true}) {texture=&T_ShieldCaustics}
  n95 = NamedRerouteDeclaration(None=Add(
    A=NamedRerouteDeclaration(None=@n90.R) {name=NoiseC, nodeColor=<rgba 0.7109 1 0 1>}
    B=NamedRerouteDeclaration(None=@n94.R) {name=NoiseD, nodeColor=<rgba 0.3359 1 0 1>}
   )) {name=NoiseC+D, nodeColor=<rgba 1 0 0.6562 1>}
  n96 = Add(A=Multiply(A=@n84, B=ScalarParameter() {defaultValue=2, parameterName=NoiseBaseScale}), B=Multiply(A=Power(Base=@n85, Exp=@n86), B=@n95))
  NamedRerouteDeclaration(None=ShadingPathSwitch(Default=ShadingPathSwitch(Default=@n96, Mobile=@n95) {desc=Migrated from FeatureLevelSwitch}, Forward=@n95, Mobile=@n95)) {name=Noise_ABCD, nodeColor=<rgba 0 1 0.2734 1>}
  n97 = Divide(A=NamedRerouteUsage(), B=ScalarParameter() {defaultValue=10, parameterName=NoiseUVDivisor})
  n98 = Sine(None=Add(
    A=Time()
    B=TextureSample(UVs=@n97) {samplerType=SAMPLERTYPE_LinearGrayscale, texture=/Engine/EngineMaterials/Good64x64TilingNoiseHighFreq}.RGB
   )) {period=0.22}
  n99 = ScalarParameter() {defaultValue=0.2, parameterName=GrainScale}
  NamedRerouteDeclaration(None=Multiply(A=Multiply(A=@n98, B=@n99), B=NamedRerouteUsage())) {name=Grain, nodeColor=<rgba 1 0 0.4922 1>}
  n100 = ComponentMask(None=Multiply(A=ReflectionVectorWS(), B=ScalarParameter() {defaultValue=2, parameterName=ChromaReflectionVectorScale})) {g=true, r=true}
  NamedRerouteDeclaration(None=Multiply(
    A=TextureSample(UVs=@n100) {texture=/Game/Global/Textures/Default/Masks_and_Gradients/3pxRGBImage}.RGB
    B=VectorParameter() {parameterName=ChromaIntensity, defaultValue=<rgba 1.8 0.2 1.8 0>}.RGB
   )) {name=Chroma, nodeColor=<rgba 0.5 0 1 1>}
  NamedRerouteUsage()
  n101 = ShadingPathSwitch(
   Default=MaterialFunctionCall(Distance=ScalarParameter() {defaultValue=200, parameterName=NearCameraFadeDistance}) {materialFunction=&Near_Camera_Fade}.Result
   Mobile=Constant() {r=1}
  ) {desc=Migrated from FeatureLevelSwitch}
  StaticSwitchParameter(True=Multiply(A=@n101, B=NamedRerouteUsage()), False=@n101) {defaultValue=true, parameterName=UseTransition}
  WorldPosition()
 ]
}
