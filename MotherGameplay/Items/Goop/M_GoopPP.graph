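# Post-process material for the Goop screen effect, applying a cool blue tint after depth of field, blended by the GoopPP collection parameter.
# At Epic quality, it also offsets scene UVs by GoopPP_Distortion and adds a masked overlay of RGB-tinted noise and slowly scaling motes.
# The noise crossfades two counter-rotated, panning layers across the screen, with radial distortion that fades as GoopPP rises.
# Other quality levels get only the tint.
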Material /MotherGameplay/Items/Goop/M_GoopPP
graph {
 properties {
  blendableLocation=BL_SceneColorAfterDOF
  materialDomain=MD_PostProcess
  pixelDepthOffsetMode=PDOM_Legacy
  shadingModel=MSM_Unlit
  usedShadingModels=Unlit
 }
 nodes [
  n1 = ComponentMask(None=SceneTexture() {sceneTextureId=PPI_PostProcessInput0}.Color) {b=true, g=true, r=true}
  n2 = Constant3Vector() {constant=<rgba 0.5391 0.8603 1.5 0>}
  n3 = CollectionParameter() {collection=&FortniteMaterialParameters, parameterName=GoopPP}
  n4 = CollectionParameter() {collection=&FortniteMaterialParameters, parameterName=GoopPP_Distortion}
  n5 = Multiply(A=NamedRerouteUsage(), B=@n4) {constB=0.1}
  n6 = ScreenPosition()
  n7 = ComponentMask(None=SceneTexture(UVs=Add(A=@n5, B=@n6.ViewportUV)) {sceneTextureId=PPI_PostProcessInput0}.Color) {b=true, g=true, r=true}
  n8 = Power(Base=OneMinus(None=NamedRerouteUsage())) {constExponent=4}
  n9 = ScreenPosition()
  n10 = Time()
  n11 = Add(A=Multiply(A=Sine(None=@n10) {period=30}) {constB=0.1})
  n12 = MaterialFunctionCall(Float2=ViewProperty() {property=MEVP_BufferSize}.Property) {materialFunction=&BreakOutFloat2Components}
  n13 = Divide(A=@n12.R, B=@n12.R)
  n14 = Multiply(
   A=Multiply(A=MaterialFunctionCall(UVs=@n9.ViewportUV, Texture Scale=@n11) {materialFunction=&ScaleUVsByCenter}.UVs) {constB=2}
   B=AppendVector(A=@n13, B=Constant() {r=1})
  )
  n15 = ComponentMask(None=ScreenPosition().ViewportUV) {r=true}
  n16 = CollectionParameter() {collection=&FortniteMaterialParameters, parameterName=GoopPP}
  n17 = ScreenPosition()
  n18 = Constant() {r=0.5}
  n19 = Constant() {r=0.5}
  n20 = Constant()
  n21 = Multiply(
   A=ConstantBiasScale(None=ScreenPosition().ViewportUV) {bias=-0.5, scale=2}
   B=NamedRerouteDeclaration(None=SphereMask(A=@n17.ViewportUV, B=@n18, Radius=@n19, Hardness=@n20)) {name=ScreenSpaceSphereMask, nodeColor=<rgba 1 0.8203 0 1>}
  )
  n22 = Add(
   A=Multiply(A=OneMinus(None=@n16), B=NamedRerouteDeclaration(None=@n21) {name=RadialDistortion, nodeColor=<rgba 0 0.4141 1 1>})
   B=ScreenPosition().ViewportUV
  )
  n23 = Rotator(Coordinate=@n22, Time=Constant() {r=0.5})
  n24 = VectorParameter() {parameterName=NoiseA_TilePan, defaultValue=<rgba 1.5 0.165 -0.094 -0.0308>}
  n25 = AppendVector(A=Multiply(A=@n13, B=@n24.R), B=@n24.G)
  n26 = TextureSample(UVs=Panner(Coordinate=Multiply(A=@n23, B=@n25), Speed=AppendVector(A=@n24.B, B=@n24.B)) {bFractionalPart=true}) {samplerType=SAMPLERTYPE_LinearColor, texture=/Game/Global/Textures/Default/Noise/T_Noise01_Inverse}
  n27 = OneMinus(None=@n15)
  n28 = Rotator(Coordinate=@n22, Time=Constant() {r=-0.5})
  n29 = VectorParameter() {parameterName=NoiseB_TilePan, defaultValue=<rgba -1.3 -0.2644 -0.0884 0.0242>}
  n30 = AppendVector(A=Multiply(A=@n13, B=@n29.R), B=@n29.G)
  n31 = TextureSample(UVs=Panner(Coordinate=Multiply(A=@n28, B=@n30), Speed=AppendVector(A=@n29.B, B=@n29.B)) {bFractionalPart=true}) {samplerType=SAMPLERTYPE_LinearColor, texture=&T_Noise01_Inverse}
  n32 = Divide(A=Add(A=Multiply(A=@n15, B=@n26.R), B=Multiply(A=@n27, B=@n31.R))) {constB=4}
  n33 = TextureSample(UVs=@n32) {texture=/Game/Global/Textures/Default/Masks_and_Gradients/3pxRGBImage}
  n34 = Add(
   A=Multiply(A=TextureSample(UVs=@n14) {samplerType=SAMPLERTYPE_LinearColor, texture=&/T_Biome_Motes}.R) {constB=0.25}
   B=LinearInterpolate(A=@n32, B=Multiply(A=@n32, B=@n33.RGB), Alpha=Multiply(A=@n8) {constB=0.5})
  )
  n35 = Add(A=Multiply(A=Multiply(A=@n8, B=@n34)) {constB=0.3}, B=@n7)
  QualitySwitch(Default=LinearInterpolate(A=@n1, B=Multiply(A=@n1, B=@n2), Alpha=@n3), Epic=LinearInterpolate(A=@n7, B=Multiply(A=@n35, B=@n2), Alpha=@n3))
 ]
}
