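# Additive, unlit, two-sided material for the goop placement preview, drawn without depth testing.
# UV edge glow, roughened by two panning caustic samples on world planes chosen by vertex color, combines with Fresnel and tints by ExternalColor.
# A scene-depth comparison desaturates and dims occluded pixels, while world-Z bands and vertex alpha modulate opacity.
# An offset toward the object pivot collapses vertices with zero alpha.
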Material /MotherGameplay/Items/Goop/M_Goop_Placement_Preview
graph {
 properties {
  bDisableDepthTest=true
  bEnableResponsiveAA=true
  blendMode=BLEND_Additive
  pixelDepthOffsetMode=PDOM_Legacy
  shadingModel=MSM_Unlit
  twoSided=true
  usedShadingModels=Unlit
 }
 nodes [
  n1 = ScalarParameter() {defaultValue=5, parameterName=EdgeScale}
  n2 = TextureCoordinate()
  n3 = MaterialFunctionCall(Float2=Abs(None=Subtract(A=@n2) {constB=0.5})) {materialFunction=&BreakOutFloat2Components}
  n4 = ScalarParameter() {defaultValue=0.004, parameterName=EdgeRadius}
  n5 = WorldPosition()
  n6 = VertexColor()
  n7 = LinearInterpolate(
   A=LinearInterpolate(A=ComponentMask(None=@n5.XYZ) {b=true, r=true}, B=ComponentMask(None=@n5.XYZ) {b=true, g=true}, Alpha=@n6.R)
   B=ComponentMask(None=@n5.XYZ) {g=true, r=true}
   Alpha=@n6.B
  )
  n8 = ScalarParameter() {defaultValue=3500, parameterName=WorldTileDivisorC}
  n9 = VectorParameter() {parameterName=NoiseC_TilePan, defaultValue=<rgba 2 2 0.0875 -0.077>}
  n10 = Panner(Coordinate=Multiply(A=Divide(A=@n7, B=@n8), B=AppendVector(A=@n9.R, B=@n9.R)), Speed=AppendVector(A=@n9.B, B=@n9.B)) {bFractionalPart=true}
  n11 = ScalarParameter() {parameterName=WorldTileDivisorC}
  n12 = VectorParameter() {parameterName=NoiseD_TilePan, defaultValue=<rgba -2 -2 0.075 0.0605>}
  n13 = Panner(Coordinate=Multiply(A=Divide(A=@n7, B=@n11), B=AppendVector(A=@n12.R, B=@n12.R)), Speed=AppendVector(A=@n12.B, B=@n12.B)) {bFractionalPart=true}
  n14 = Add(
   A=NamedRerouteDeclaration(None=TextureSample(UVs=@n10) {texture=&T_ShieldCaustics}.R) {name=NoiseC, nodeColor=<rgba 0.7109 1 0 1>}
   B=NamedRerouteDeclaration(None=TextureSample(UVs=@n13) {texture=&T_ShieldCaustics}.R) {name=NoiseD, nodeColor=<rgba 0.3359 1 0 1>}
  )
  n15 = Multiply(A=Subtract(A=@n14) {constB=0.25}, B=ScalarParameter() {defaultValue=0.008, parameterName=EdgeNoiseScale})
  n16 = SphereMask(
   A=Max(A=@n3.R, B=@n3.R)
   B=Constant() {r=0.5}
   Radius=Max(A=ScalarParameter() {defaultValue=0.002, parameterName=EdgeRadiusMin}, B=Add(A=@n4, B=@n15))
   Hardness=LinearInterpolate(B=ScalarParameter() {defaultValue=0.4, parameterName=EdgeHardness}, Alpha=@n14)
  )
  n17 = Add(
   A=Multiply(A=ScalarParameter() {defaultValue=0.5, parameterName=EdgeBaseSubtleScale}, B=@n16)
   B=Power(Base=@n16, Exp=ScalarParameter() {defaultValue=8, parameterName=EdgeExponent})
  )
  n18 = Multiply(
   A=Add(A=Multiply(A=@n1, B=@n17), B=Fresnel() {exponent=3})
   B=VectorParameter() {parameterName=ExternalColor, defaultValue=<rgba 0.0859 0.5913 1.5 1>}.RGB
  )
  n19 = SceneDepth()
  n20 = PixelDepth()
  n21 = ConstantBiasScale(None=Sign(None=Subtract(A=@n19.R, B=@n20.R)))
  LinearInterpolate(A=Desaturation(None=@n18, Fraction=Constant() {r=1}), B=@n18, Alpha=@n21)
  n22 = Multiply(A=ComponentMask(None=WorldPosition().XYZ) {b=true}) {constB=0.32}
  n23 = Floor(None=Add(A=Frac(None=@n22)) {constB=0.2})
  n24 = Multiply(A=Saturate(None=Add(A=@n23) {constB=0.4}), B=VertexColor().A)
  LinearInterpolate(A=Multiply(A=@n24, B=ScalarParameter() {defaultValue=0.05, parameterName=BehindOpaqueTransparencyScale}) {constB=0.2}, B=@n24, Alpha=@n21)
  n25 = Multiply(A=VertexColor().A)
  Multiply(
   A=Subtract(
    A=MaterialFunctionCall() {materialFunction=&ObjectPivotPoint}.Object Pivot Location
    B=WorldPosition().XYZ
   )
   B=OneMinus(None=Ceil(None=@n25))
  )
 ]
}
