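; Goop projectile that places a goop building gameplay actor (BGA_Goop) where it stops.
; On stop it spawns the BGA at its building-grid-snapped location with a downward offset, credited to the instigator, then kills itself.
; Cosmetic-only steps fade the looping tell on hit or stop; stopping also hides the mesh and flags the Niagara trail to die out.
; An equip/aim sound attached to the instigator starts at BeginPlay and fades at EndPlay.
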(event EventBeginPlay
  (|CallParentFunction)
  (bind _returnvalue (Audio|SpawnSoundAttached "/MotherGameplay/Items/Goop/Audio/Goop_Equip_Aim_Cue.Goop_Equip_Aim_Cue" (Variables|/Script/Engine.Actor|GetRootComponent (Game|GetInstigator))))
  (Variables|Default|SetEquipAudioComponent _returnvalue))

(event EventEndPlay
  (params (EndPlayReason :type EEndPlayReason%20Enum :default "Destroyed"))
  (|CallParentFunction EndPlayReason)
  (bind _equipaudiocomponent (Variables|Default|GetEquipAudioComponent)
    (:then
      (Audio|Components|Audio|FadeOut _equipaudiocomponent 0.5 0.0 "Linear" :!class /Script/Engine.AudioComponent))
    (:else)))

(event EventOnStop
  (params (Hit :type Hit%20Result%20Structure%20%28by%20ref%29))
  ;[1 1 1] On Stop - Spawn BGA
  (|CallParentFunction Hit)
  (Building|SpawnBuildingGameplayActorfromWorldContext (Variables|Default|GetBGA_Goop) (Math|Transform|MakeTransform (Fortnite|SnaptoFortBuildingGrid (Transformation|GetActorLocation self) "0.000000,0.000000,-192.000000")) (Variables|Actor|GetInstigator) (Variables|Actor|GetInstigator))
  (Delivery|Kill)
  (Networking|CanExecuteCosmeticEvents
    (:True
      (Audio|Components|Audio|FadeOut (Variables|Default|GetLoopingAudioTell) 0.25 0.0 "Linear" :!class /Script/Engine.AudioComponent)
      ;[1 1 1] Let trail fadeout
      (Effects|Components|ParticleSystem|SetBoolParameter (Variables|Default|GetNS_Biome_Projectile_Trail) "OnDeath" true)
      (Development|SetHiddeninGame (Variables|Default|GetMesh) true true))))

(event Collision|EventHit
  (params (MyComp :type Primitive%20Component%20Object%20Reference) (Other :type Actor%20Object%20Reference) (OtherComp :type Primitive%20Component%20Object%20Reference) (bSelfMoved :type Boolean) (HitLocation :type Vector :default "0, 0, 0") (HitNormal :type Vector :default "0, 0, 0") (NormalImpulse :type Vector :default "0, 0, 0") (Hit :type Hit%20Result%20Structure%20%28by%20ref%29))
  ;[1 1 1] On Exploded
  (|CallParentFunction MyComp Other OtherComp bSelfMoved HitLocation HitNormal NormalImpulse Hit)
  (Networking|CanExecuteCosmeticEvents
    (:True
      (Audio|Components|Audio|FadeOut (Variables|Default|GetLoopingAudioTell) 0.25 0.0 "Linear" :!class /Script/Engine.AudioComponent))))

(fn ConstructionScript)
